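# Translucent UI material for an angled bracket frame decoration, with adaptive or custom color.
# Texture R and G channels become anti-aliased distance-field shapes, faded by opposing vertical gradients and blended by SwitchToBottom.
# A separate procedural branch cuts a perspective-skewed box from a three-sided polygon SDF, rotated in quarter turns by PickCorner.
# HasAdaptiveColors selects CustomColor or adaptive bright color, which can shift to the shared parameter-collection yellow.
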Material /BattleRoyaleGameUI/Shared/FrameDecor/M_UI_AngledBracket
graph {
 properties {bAutomaticallySetUsageInEditor=false, blendMode=BLEND_Translucent, materialDomain=MD_UI, usedShadingModels=Default Lit}
 nodes [
  Constant() {r=0.037}
  n1 = TextureSample() {texture=&/T_UI_CornerT_UI_BR_AngledBracket_DF}
  n2 = Constant() {r=0.2}
  n3 = OneMinus(None=MaterialFunctionCall() {materialFunction=&LinearGradient}.VGradient)
  n4 = Constant() {r=0.2}
  n5 = MaterialFunctionCall() {materialFunction=&LinearGradient}
  LinearInterpolate(
   A=Multiply(
    A=MaterialFunctionCall(DistanceField=@n1.R, Threshold=@n2) {materialFunction=&MF_UI_AntiAliasing}.Outer
    B=Power(Base=@n3) {constExponent=4}
   )
   B=Multiply(
    A=MaterialFunctionCall(DistanceField=@n1.G, Threshold=@n4) {materialFunction=&MF_UI_AntiAliasing}.Outer
    B=Power(Base=@n5.VGradient) {constExponent=4}
   )
   Alpha=ScalarParameter() {parameterName=SwitchToBottom}
  )
  MaterialFunctionCall!(MSDF RGB) {materialFunction=/FortUILibrary/MaterialFunctions/Shapes/Modifiers/MF_UI_MSDFtoSDF}
  MaterialFunctionCall() {materialFunction=&MF_UI_MSDFtoSDF}
  n6 = Multiply(A=Subtract(A=ScalarParameter() {defaultValue=4, parameterName=PickCorner, sliderMax=4, sliderMin=1})) {constB=0.25}
  n7 = MaterialFunctionCall(UVs=MaterialFunctionCall() {materialFunction=&GetUserInterfaceUV}.Normalized UV, "Rotation Angle (0-1)"=@n6) {materialFunction=&CustomRotator}
  n8 = Constant()
  n9 = Add(
   A=MaterialFunctionCall!(
    UV=MaterialFunctionCall(UVs=@n7.Rotated Values, "Rotation Angle (0-1)"=@n8) {materialFunction=&CustomRotator}.Rotated Values
    Perspective Rotation=Constant() {r=0.115}
    Perspective Amount=Constant() {r=0.4}
   ) {materialFunction=/FortUILibrary/MaterialFunctions/Basics/MF_UI_Perspective}.Perspective UVs
   B=Constant2Vector() {g=0.05}
  )
  n10 = Constant() {r=1.75}
  n11 = Constant()
  n12 = MaterialFunctionCall(
   UVs=MaterialFunctionCall(UVs=MaterialFunctionCall() {materialFunction=&GetUserInterfaceUV}.Normalized UV, "Rotation Angle (0-1)"=@n6) {materialFunction=&CustomRotator}.Rotated Values
   "Rotation Angle (0-1)"=Constant() {r=0.035}
  ) {materialFunction=&CustomRotator}
  n13 = Constant2Vector()
  n14 = MaterialFunctionCall!(
   UVs=MaterialFunctionCall(UV=Subtract(A=@n12.Rotated Values, B=@n13), Perspective Rotation=Constant() {r=0.115}, Perspective Amount=Constant() {r=0.48}) {materialFunction=&MF_UI_Perspective}.Perspective UVs
   Radius=Constant() {r=5}
   Sides=Constant() {r=3}
   InnerAngle
  ) {materialFunction=/FortUILibrary/MaterialFunctions/Shapes/SDFs/MF_SDF_RegularPolygon}
  Multiply(
   A=OneMinus(None=MaterialFunctionCall(UVs=@n9, Size=@n10, CornerRoundness=@n11) {materialFunction=&MF_UI_Box}.Fill)
   B=NamedRerouteDeclaration(None=MaterialFunctionCall(DistanceField=@n14.SDF) {materialFunction=&MF_UI_AntiAliasing}.Inner) {name=Name_2, nodeColor=<rgba 1 0.5156 0 1>}
  )
  Multiply(A=Subtract(A=ScalarParameter() {defaultValue=4, parameterName=PickCorner, sliderMax=4, sliderMin=1})) {constB=0.25}
  n15 = MaterialFunctionCall!(BufferSource) {materialFunction=/BattleRoyaleGameUI/Shared/MF_UI_AdaptiveColors}
  n16 = LinearInterpolate(A=@n15.Bright, B=@n15.Bright, Alpha=ScalarParameter() {parameterName=Anim_Color})
  n17 = CollectionParameter() {collection=/BattleRoyaleGameUI/Shared/MPC_BRGameUI, parameterName=Yellow}
  StaticSwitchParameter(
   True=LinearInterpolate(
    A=StaticSwitchParameter(True=@n16, False=@n15.Bright) {parameterName=IsAnimated}
    B=ComponentMask(None=@n17) {b=true, g=true, r=true}
    Alpha=ScalarParameter() {parameterName=MPCYellow}
   )
   False=VectorParameter() {parameterName=CustomColor, defaultValue=<rgba 1 1 1 1>}.RGB
  ) {parameterName=HasAdaptiveColors}
 ]
}
