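# Translucent UI material that draws a bolt-shaped frame decoration for the Battle Royale game UI, with FlipX and FlipY UV controls.
# The shape is an anti-aliased SDF made by subtracting a notch triangle from a main triangle, both with parameterized corners.
# Static switches select animated or fixed color, drawing on adaptive palette shades, a custom color or MPC_BRGameUI red, which IsElim can blend in.
# Opacity is animated or fixed, with low-end device adjustment.
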Material /BattleRoyaleGameUI/Shared/FrameDecor/M_UI_Bolt
graph {
 properties {bAutomaticallySetUsageInEditor=false, blendMode=BLEND_Translucent, materialDomain=MD_UI, numCustomizedUVs=1, usedShadingModels=Default Lit}
 nodes [
  n1 = MaterialFunctionCall!(
   DistanceFieldA=MaterialFunctionCall!(
    UVs
    Corner 01=VectorParameter() {group=Shape, parameterName=Corner1_Position, defaultValue=<rgba -1 -1 0 1>}.RGB
    Corner 02=VectorParameter() {group=Shape, parameterName=Corner2_Position, defaultValue=<rgba 1 -1 0 1>}.RGB
    Corner 03=VectorParameter() {group=Shape, parameterName=Corner3_Position, defaultValue=<rgba -1 1 0 1>}.RGB
    Glow Min
    Glow Max
    Outline Thickness
   ) {materialFunction=/FortUILibrary/MaterialFunctions/Shapes/MF_UI_WIP_Triangle}.SDF
   DistanceFieldB=MaterialFunctionCall(
    Corner 01=VectorParameter() {group=Shape, parameterName=Notch_Corner1_Position, defaultValue=<rgba 1 -1 0 1>}.RGB
    Corner 02=VectorParameter() {group=Shape, parameterName=Notch_Corner2_Position, defaultValue=<rgba 1 0.5 0 1>}.RGB
    Corner 03=VectorParameter() {group=Shape, parameterName=NotchCorner3_Position, defaultValue=<rgba -0.75 0 0 1>}.RGB
   ) {materialFunction=&MF_UI_WIP_Triangle}.SDF
   BlendAmount
  ) {materialFunction=/FortUILibrary/MaterialFunctions/Shapes/Modifiers/MF_UI_SDF_Subtraction}
  n2 = ScalarParameter() {defaultValue=0.005, group=Shape, parameterName=EdgeSubtraction}
  n3 = Constant() {r=1}
  n4 = ScalarParameter() {defaultValue=0.005, group=Shape, parameterName=EdgeSubtraction}
  n5 = Constant() {r=1}
  n6 = Constant() {r=-0.25}
  MaterialFunctionCall(
   DistanceFieldA=@n1.Exact
   DistanceFieldB=MaterialFunctionCall!(
    DistanceFieldA=MaterialFunctionCall(GradientStart=@n2, GradientEnd=@n3) {materialFunction=&MF_UI_LinearGradient}.Result
    DistanceFieldB=MaterialFunctionCall(GradientStart=@n4, GradientEnd=@n5, RotationAngle=@n6) {materialFunction=&MF_UI_LinearGradient}.Result
    BlendAmount
   ) {materialFunction=/FortUILibrary/MaterialFunctions/Shapes/Modifiers/MF_UI_SDF_Union}.Exact
  ) {materialFunction=&MF_UI_SDF_Subtraction}
  n7 = MaterialFunctionCall!(BufferSource) {materialFunction=/BattleRoyaleGameUI/Shared/MF_UI_AdaptiveColors}
  n8 = ScalarParameter() {defaultValue=1, group=Color, parameterName=AdaptiveColorIndex}
  n9 = LinearInterpolate(A=@n7.Dark, B=@n7.Dark, Alpha=@n8)
  n10 = Saturate(None=Subtract(A=@n8))
  n11 = Subtract(A=@n8) {constB=2}
  NamedRerouteDeclaration(None=LinearInterpolate(A=LinearInterpolate(A=@n9, B=@n7.Medium, Alpha=@n10), B=@n7.Bright, Alpha=Saturate(None=@n11))) {name=AdaptiveColorOutput, nodeColor=<rgba 0 0 0 1>}
  n12 = MaterialFunctionCall() {materialFunction=&GetUserInterfaceUV}
  n13 = MaterialFunctionCall(Float2=@n12.Normalized UV) {materialFunction=&BreakOutFloat2Components}
  n14 = AppendVector(A=OneMinus(None=@n13.R), B=@n13.G)
  n15 = ScalarParameter() {group=FlipUVs, parameterName=FlipX}
  n16 = OneMinus(None=@n13.G)
  NamedRerouteDeclaration(None=LinearInterpolate(
    A=LinearInterpolate(A=@n12.Normalized UV, B=@n14, Alpha=@n15)
    B=AppendVector(A=@n13.R, B=@n16)
    Alpha=ScalarParameter() {group=FlipUVs, parameterName=FlipY}
   )) {name=FlipUVs, nodeColor=<rgba 0 0 0 1>}
  NamedRerouteDeclaration(None=MaterialFunctionCall(DistanceField=@n1.Exact) {materialFunction=&MF_UI_AntiAliasing}.Inner) {name=Shape, nodeColor=<rgba 0 0 0 1>}
  n17 = ScalarParameter() {defaultValue=1, group=Opacity, parameterName=Opacity_Stop}
  n18 = LinearInterpolate(
   A=ScalarParameter() {group=Opacity, parameterName=Opacity_Start}
   B=MaterialFunctionCall!(Default=@n17, LowEnd=Multiply(A=@n17) {constB=0.9}) {materialFunction=/BattleRoyaleGameUI/Shared/MF_UI_LowEnd_ValueSwitcher}.Result
   Alpha=ScalarParameter() {defaultValue=1, group=Opacity, parameterName=Anim_Opacity}
  )
  n19 = ScalarParameter() {defaultValue=1, group=Opacity, parameterName=Opacity}
  n20 = MaterialFunctionCall(Default=@n19, LowEnd=Multiply(A=@n19) {constB=0.9}) {materialFunction=&MF_UI_LowEnd_ValueSwitcher}
  NamedRerouteDeclaration(None=Multiply(A=StaticSwitchParameter(True=@n18, False=@n20.Result) {defaultValue=true, parameterName=AnimatedOpacity}, B=NamedRerouteUsage())) {name=Opacity, nodeColor=<rgba 0 0 0 1>}
  n21 = ScalarParameter() {parameterName=IsElim}
  n22 = ComponentMask(None=CollectionParameter() {collection=/BattleRoyaleGameUI/Shared/MPC_BRGameUI, parameterName=Red}) {b=true, g=true, r=true}
  n23 = LinearInterpolate(
   A=LinearInterpolate(A=@n7.Dim, B=@n7.Dim, Alpha=@n21)
   B=LinearInterpolate(A=@n7.White, B=@n22, Alpha=@n21)
   Alpha=ScalarParameter() {group=Color, parameterName=Anim_Color}
  )
  n24 = NamedRerouteUsage()
  n25 = StaticSwitchParameter(True=VectorParameter() {group=Color, parameterName=Color, defaultValue=<rgba 1 0 0 1>}.RGB, False=@n22) {parameterName=CustomColor}
  n26 = LinearInterpolate(A=@n24, B=@n25, Alpha=ScalarParameter() {group=Color, parameterName=Anim_Color})
  NamedRerouteDeclaration(None=StaticSwitchParameter(
    True=StaticSwitchParameter(True=@n23, False=@n26) {defaultValue=true, parameterName=AdaptiveColorAnimation}
    False=StaticSwitchParameter(True=@n24, False=@n25) {defaultValue=true, parameterName=AdaptiveColor}
   ) {defaultValue=true, parameterName=AnimatedColors}) {name=Color, nodeColor=<rgba 0 0 0 1>}
  NamedRerouteUsage()
  NamedRerouteUsage()
  NamedRerouteUsage()
 ]
}
