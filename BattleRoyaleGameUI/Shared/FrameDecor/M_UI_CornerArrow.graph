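# Translucent UI material that draws a corner arrow decoration for Battle Royale UI frames, with PickCorner rotating it in quarter turns.
# The shape is procedural: rotated UVs pass through a perspective warp and offset into an inverted box mask scaled by Opacity.
# IsAnimated adds scale, opacity and color parameters, though the color lerp blends Bright with itself.
# HasAdaptiveColors selects adaptive Bright, blendable to the shared collection's Yellow, or else CustomColor.
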Material /BattleRoyaleGameUI/Shared/FrameDecor/M_UI_CornerArrow
graph {
 properties {bAutomaticallySetUsageInEditor=false, blendMode=BLEND_Translucent, materialDomain=MD_UI, usedShadingModels=Default Lit}
 nodes [
  n1 = ScalarParameter() {defaultValue=1, parameterName=PickCorner, sliderMax=4, sliderMin=1}
  n2 = MaterialFunctionCall(
   UVs=MaterialFunctionCall() {materialFunction=&GetUserInterfaceUV}.Normalized UV
   "Rotation Angle (0-1)"=Multiply(A=Subtract(A=@n1)) {constB=0.25}
  ) {materialFunction=&CustomRotator}
  n3 = Constant()
  n4 = Add(
   A=MaterialFunctionCall!(
    UV=MaterialFunctionCall(UVs=@n2.Rotated Values, "Rotation Angle (0-1)"=@n3) {materialFunction=&CustomRotator}.Rotated Values
    Perspective Rotation=Constant() {r=0.115}
    Perspective Amount=Constant() {r=0.4}
   ) {materialFunction=/FortUILibrary/MaterialFunctions/Basics/MF_UI_Perspective}.Perspective UVs
   B=Constant2Vector() {g=0.05}
  )
  n5 = Constant() {r=1.75}
  n6 = StaticSwitchParameter(True=LinearInterpolate(A=@n5, Alpha=ScalarParameter() {parameterName=Anim_Scale}) {constB=32}, False=@n5) {parameterName=IsAnimated}
  n7 = Constant()
  n8 = Multiply(
   A=OneMinus(None=MaterialFunctionCall(UVs=@n4, Size=@n6, CornerRoundness=@n7) {materialFunction=&MF_UI_Box}.Fill)
   B=ScalarParameter() {defaultValue=0.5, parameterName=Opacity}
  )
  StaticSwitchParameter(True=LinearInterpolate(B=@n8, Alpha=ScalarParameter() {defaultValue=1, parameterName=Anim_Opacity}), False=@n8) {parameterName=IsAnimated}
  n9 = MaterialFunctionCall!(BufferSource) {materialFunction=/BattleRoyaleGameUI/Shared/MF_UI_AdaptiveColors}
  n10 = LinearInterpolate(A=@n9.Bright, B=@n9.Bright, Alpha=ScalarParameter() {parameterName=Anim_Color})
  n11 = CollectionParameter() {collection=/BattleRoyaleGameUI/Shared/MPC_BRGameUI, parameterName=Yellow}
  StaticSwitchParameter(
   True=LinearInterpolate(
    A=StaticSwitchParameter(True=@n10, False=@n9.Bright) {parameterName=IsAnimated}
    B=ComponentMask(None=@n11) {b=true, g=true, r=true}
    Alpha=ScalarParameter() {parameterName=MPCYellow}
   )
   False=VectorParameter() {parameterName=CustomColor, defaultValue=<rgba 1 1 1 1>}.RGB
  ) {parameterName=HasAdaptiveColors}
 ]
}
