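# Translucent UI material that draws the quickbar tab bracket frame decoration in selected and unselected states.
# Builds the shape by subtracting two trapezoid SDFs in swizzled 9-slice UVs, one skewed, with IsSelected blending the skew and tab config.
# The SDF feeds an antialiased body with per-state vertical opacity fades and a soft glow, which the IsOnlyGlow static switch can isolate.
# Adaptive Dim and Bright colors supply the tint.
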Material /BattleRoyaleGameUI/Shared/FrameDecor/M_UI_QuickbarTabBracket
graph {
 properties {bAutomaticallySetUsageInEditor=false, blendMode=BLEND_Translucent, materialDomain=MD_UI, usedShadingModels=Default Lit}
 nodes [
  n1 = MaterialFunctionCall!(BufferSource) {materialFunction=/BattleRoyaleGameUI/Shared/MF_UI_AdaptiveColors}
  n2 = NamedRerouteUsage()
  n3 = Saturate(None=MaterialFunctionCall("Value (0-1)"=OneMinus(None=@n2)) {materialFunction=&MF_UI_EaseCurves}.EaseInSine)
  LinearInterpolate(
   A=@n1.Dim
   B=LinearInterpolate(A=@n1.Dim, B=LinearInterpolate(A=@n1.Bright, B=@n1.Bright, Alpha=@n3), Alpha=ScalarParameter() {parameterName=IsSelected})
   Alpha=NamedRerouteUsage()
  )
  n4 = VectorParameter() {parameterName=DefaultConfig, defaultValue=<rgba 340 0.5 0.75 1>}
  n5 = VectorParameter() {parameterName=SelectedConfig, defaultValue=<rgba 7.5 0.333 0.875 0>}
  n6 = MaterialFunctionCall(Float4=LinearInterpolate(A=AppendVector(A=@n4.RGB, B=@n4.RGB), B=AppendVector(A=@n5.RGB, B=@n5.RGB), Alpha=ScalarParameter() {parameterName=IsSelected})) {materialFunction=&BreakOutFloat4Components}
  NamedRerouteDeclaration(None=@n6.G) {name=TabLeftHeight, nodeColor=<rgba 0 0 0 1>}
  NamedRerouteDeclaration(None=@n6.B) {name=TabRightHeight, nodeColor=<rgba 0 0 0 1>}
  NamedRerouteDeclaration(None=@n6.A) {name=TabOffset, nodeColor=<rgba 0 0 0 1>}
  n7 = MaterialFunctionCall(XY=MaterialFunctionCall!(UVs) {materialFunction=/FortUILibrary/MaterialFunctions/Basics/MF_UI_9SliceUVs}.UVs) {materialFunction=&Swizzle}
  n8 = ScalarParameter() {defaultValue=-16, parameterName=UnselectedSkew}
  n9 = ScalarParameter() {defaultValue=7.5, parameterName=SelectedSkew}
  n10 = ScalarParameter() {parameterName=IsSelected}
  n11 = MaterialFunctionCall!(UV=@n7.YX, Skew Angle=AppendVector(A=LinearInterpolate(A=@n8, B=@n9, Alpha=@n10) {constA=-16, constB=7.5}, B=Constant())) {materialFunction=/FortUILibrary/MaterialFunctions/Basics/MF_UI_Skew}
  n12 = Constant2Vector() {g=-0.1}
  n13 = AppendVector(A=Multiply(A=NamedRerouteUsage()) {constB=-1}, B=ScalarParameter() {defaultValue=-0.3, parameterName=Param_3})
  n14 = ScalarParameter() {defaultValue=0.5, parameterName=Param_2}
  n15 = Multiply(A=NamedRerouteUsage()) {constB=2}
  n16 = ScalarParameter() {defaultValue=0.5, parameterName=Param_1}
  n17 = MaterialFunctionCall!(
   DistanceFieldA=MaterialFunctionCall!(
    UVs=Subtract(A=@n11.Skewed UVs, B=@n12)
    TopWidth=NamedRerouteUsage()
    BottomWidth=NamedRerouteUsage()
    Height=ScalarParameter() {defaultValue=0.6}
   ) {materialFunction=/FortUILibrary/MaterialFunctions/Shapes/SDFs/MF_SDF_IsoscelesTrapezoid}.SDF
   DistanceFieldB=MaterialFunctionCall(
    UVs=Subtract(A=@n7.YX, B=@n13)
    TopWidth=Add(A=@n14, B=@n15)
    BottomWidth=Add(A=@n16, B=@n15)
    Height=ScalarParameter() {defaultValue=0.6, parameterName=Param_4}
   ) {materialFunction=&MF_SDF_IsoscelesTrapezoid}.SDF
   BlendAmount
  ) {materialFunction=/FortUILibrary/MaterialFunctions/Shapes/Modifiers/MF_UI_SDF_Subtraction}
  n18 = Multiply(
   A=MaterialFunctionCall(
    DistanceField=@n17.Smooth
    Threshold=ScalarParameter() {defaultValue=-0.04, parameterName=GlowThreshold}
    Smoothing=ScalarParameter() {defaultValue=0.25, parameterName=GlowSmoothing}
   ) {materialFunction=&MF_UI_AntiAliasing}.Inner
   B=Multiply(A=ScalarParameter() {defaultValue=0.5, parameterName=SelectedGlowAmount}, B=NamedRerouteUsage())
  )
  n19 = ScalarParameter() {parameterName=UnselectedOpacity_End}
  n20 = ScalarParameter() {defaultValue=1, parameterName=UnselectedOpacity_Start}
  n21 = MaterialFunctionCall() {materialFunction=&GetUserInterfaceUV}
  n22 = NamedRerouteDeclaration(None=OneMinus(None=ComponentMask(None=@n21.Normalized UV) {g=true})) {name=Fade, nodeColor=<rgba 0 0 0 1>}
  n23 = LinearInterpolate(
   A=ScalarParameter() {defaultValue=1, parameterName=SelectedOpacity_Stop}
   B=ScalarParameter() {defaultValue=1, parameterName=SelectedOpacity_Start}
   Alpha=@n22
  )
  n24 = Multiply(
   A=MaterialFunctionCall(DistanceField=@n17.Exact) {materialFunction=&MF_UI_AntiAliasing}.Inner
   B=LinearInterpolate(
    A=LinearInterpolate(A=@n19, B=@n20, Alpha=@n22)
    B=NamedRerouteDeclaration(None=@n23) {name=SelectedFade, nodeColor=<rgba 0.5234 1 0 1>}
    Alpha=ScalarParameter() {parameterName=IsSelected}
   )
  )
  n25 = LinearInterpolate(A=@n18, Alpha=@n24)
  StaticSwitchParameter(
   True=NamedRerouteDeclaration(None=StaticSwitchParameter(True=@n18, False=@n25) {parameterName=IsOnlyGlow}) {name=FinalMask, nodeColor=<rgba 0 0 0 1>}
   False=LinearInterpolate(A=@n24, B=NamedRerouteUsage(), Alpha=ScalarParameter() {parameterName=IsSelected})
  ) {parameterName=IsOnlyGlow}
 ]
}
