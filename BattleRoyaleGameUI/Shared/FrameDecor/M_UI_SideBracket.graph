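# Translucent UI material that draws a procedural side-bracket frame decoration with optional glow for Battle Royale UI.
# Rotated, mirrored 9-slice UVs feed a three-sided polygon SDF, from which a trapezoid SDF is subtracted for an antialiased shape and soft glow mask.
# Static switches select glow, glow-only and custom color variants.
# Scalar parameters animate opacity and blend custom gradients with shared MPC blue and cyan and the adaptive Dim color.
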Material /BattleRoyaleGameUI/Shared/FrameDecor/M_UI_SideBracket
graph {
 properties {bAutomaticallySetUsageInEditor=false, blendMode=BLEND_Translucent, materialDomain=MD_UI, usedShadingModels=Default Lit}
 nodes [
  n1 = MaterialFunctionCall!(UVs) {materialFunction=/FortUILibrary/MaterialFunctions/Basics/MF_UI_9SliceUVs}
  n2 = Constant() {r=0.25}
  n3 = Add(
   A=MaterialFunctionCall!(
    UV=MaterialFunctionCall(UVs=@n1.UVs, "Rotation Angle (0-1)"=@n2) {materialFunction=&CustomRotator}.Rotated Values
    Mirror Y
    Swizzle UV
   ) {materialFunction=/FortUILibrary/MaterialFunctions/Basics/MF_UI_Symmetry_2x1_New}.Symmetry
   B=Constant2Vector() {r=-0.5}
  )
  n4 = Constant2Vector() {g=1.3333, r=2.5}
  n5 = Subtract(A=@n1.UVs, B=Constant2Vector() {r=-0.435})
  n6 = MaterialFunctionCall!(
   DistanceFieldA=MaterialFunctionCall!(
    UVs=MaterialFunctionCall(UVs=@n3, Texture Scale=@n4) {materialFunction=&ScaleUVsByCenter}.UVs
    Radius=ScalarParameter() {defaultValue=0.9, parameterName=Radius}
    Sides=Constant() {r=3}
    InnerAngle
   ) {materialFunction=/FortUILibrary/MaterialFunctions/Shapes/SDFs/MF_SDF_RegularPolygon}.SDF
   DistanceFieldB=MaterialFunctionCall!(
    UVs=MaterialFunctionCall(XY=@n5) {materialFunction=&Swizzle}.YX
    TopWidth=ScalarParameter() {defaultValue=0.45, parameterName=TopWidth}
    BottomWidth=ScalarParameter() {defaultValue=0.45, parameterName=BottomWidth}
    Height=ScalarParameter() {defaultValue=0.95, parameterName=Height}
   ) {materialFunction=/FortUILibrary/MaterialFunctions/Shapes/SDFs/MF_SDF_IsoscelesTrapezoid}.SDF
   BlendAmount
  ) {materialFunction=/FortUILibrary/MaterialFunctions/Shapes/Modifiers/MF_UI_SDF_Subtraction}
  n7 = SmoothStep(Min=ScalarParameter() {defaultValue=0.8, parameterName=GlowDistance}, Value=OneMinus(None=@n6.Smooth)) {constMin=0.8}
  n8 = NamedRerouteDeclaration(None=MaterialFunctionCall(DistanceField=@n6.Exact) {materialFunction=&MF_UI_AntiAliasing}.Inner) {name=ShapeMask, nodeColor=<rgba 0 0 0 1>}
  NamedRerouteDeclaration(None=LinearInterpolate(A=NamedRerouteDeclaration(None=@n7) {name=GlowMask, nodeColor=<rgba 0 0 0 1>}, Alpha=@n8)) {name=WithGlowMask, nodeColor=<rgba 0 0 0 1>}
  n9 = Multiply(
   A=NamedRerouteUsage()
   B=LinearInterpolate(
    A=ScalarParameter() {defaultValue=1, parameterName=GlowOpacity}
    B=ScalarParameter() {defaultValue=1, parameterName=GlowOpacity_AnimStop}
    Alpha=ScalarParameter() {parameterName=Anim_Opacity}
   ) {constB=0.95}
  )
  n10 = Multiply(
   A=@n8
   B=LinearInterpolate(A=ScalarParameter() {defaultValue=0.5, parameterName=Opacity}, Alpha=ScalarParameter() {parameterName=Anim_Opacity}) {constB=0.95}
  )
  StaticSwitchParameter(True=@n9, False=StaticSwitchParameter(True=LinearInterpolate(A=@n9, B=@n10, Alpha=@n8), False=@n10) {parameterName=HasGlow}) {parameterName=IsOnlyGlow}
  n11 = VectorParameter() {parameterName=GlowColor, defaultValue=<rgba 1 1 1 1>}
  n12 = ComponentMask(None=TextureCoordinate()) {g=true}
  n13 = ComponentMask(None=CollectionParameter() {collection=/BattleRoyaleGameUI/Shared/MPC_BRGameUI, parameterName=Blue}) {b=true, g=true, r=true}
  n14 = VectorParameter() {parameterName=UnselectedColor_A, defaultValue=<rgba 0 0 1 1>}
  n15 = ScalarParameter() {defaultValue=0.75, parameterName=DimBlueHighlightColor}
  n16 = LinearInterpolate(
   A=LinearInterpolate(
    A=VectorParameter() {parameterName=Color, defaultValue=<rgba 1 1 1 1>}.RGB
    B=VectorParameter() {parameterName=Color_2, defaultValue=<rgba 1 1 1 1>}.RGB
    Alpha=@n12
   )
   B=LinearInterpolate(
    A=LinearInterpolate(A=@n13, B=@n14.RGB, Alpha=@n15) {constAlpha=0.75}
    B=@n14.RGB
    Alpha=MaterialFunctionCall("Value (0-1)"=@n12) {materialFunction=&MF_UI_EaseCurves}.EaseOutSine
   )
   Alpha=ScalarParameter() {parameterName=MPCBlue}
  )
  n17 = MaterialFunctionCall!(BufferSource) {materialFunction=/BattleRoyaleGameUI/Shared/MF_UI_AdaptiveColors}
  n18 = LinearInterpolate(
   A=StaticSwitchParameter(True=LinearInterpolate(A=@n11.RGB, B=@n16), False=@n16) {parameterName=HasGlow}
   B=@n17.Dim
   Alpha=ScalarParameter() {parameterName=CyanToDimAnim}
  )
  n19 = ComponentMask(None=CollectionParameter() {collection=&MPC_BRGameUI, parameterName=Cyan}) {b=true, g=true, r=true}
  n20 = ScalarParameter() {parameterName=Anim_Color}
  n21 = ScalarParameter() {parameterName=SelectPostProcessBright}
  StaticSwitchParameter(
   True=VectorParameter() {parameterName=GlowColor, defaultValue=<rgba 1 1 1 1>}.RGB
   False=StaticSwitchParameter(True=LinearInterpolate(A=@n18, B=@n19, Alpha=@n20), False=LinearInterpolate(A=@n17.Dim, B=@n17.Dim, Alpha=@n21)) {parameterName=CustomColor}
  ) {parameterName=IsOnlyGlow}
 ]
}
