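# Translucent UI material that draws a simple horizontal divider with a soft shadow for Battle Royale UI frame decor.
# Offset UI UVs pass through a perspective warp into an isosceles trapezoid function with parameterized widths and height.
# The trapezoid fill blends adaptive dark or custom shadow color with the shape color.
# The SDF yields a smoothed, opacity-scaled shadow mask, which is blended with the fill into a combined mask.
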Material /BattleRoyaleGameUI/Shared/FrameDecor/M_UI_SimpleDivider_Horizontal
graph {
 properties {bAutomaticallySetUsageInEditor=false, blendMode=BLEND_Translucent, materialDomain=MD_UI, usedShadingModels=Default Lit}
 nodes [
  CollectionParameter() {collection=/BattleRoyaleGameUI/Shared/MPC_BRGameUI, parameterName=FrameColor}
  n1 = MaterialFunctionCall() {materialFunction=&GetUserInterfaceUV}
  n2 = Constant2Vector() {r=-0.08}
  n3 = MaterialFunctionCall!(
   UVs=MaterialFunctionCall!(
    UV=Subtract(A=@n1.Normalized UV, B=@n2)
    Perspective Rotation=ScalarParameter() {parameterName=PerspectiveRotation}
    Perspective Amount=ScalarParameter() {defaultValue=0.3, parameterName=PerspectiveAmount}
   ) {materialFunction=/FortUILibrary/MaterialFunctions/Basics/MF_UI_Perspective}.Perspective UVs
   TopWidth=ScalarParameter() {defaultValue=0.9, parameterName=ShadeTopWidth}
   BottomWidth=ScalarParameter() {defaultValue=0.75, parameterName=ShapeBottomWidth}
   Height=ScalarParameter() {defaultValue=0.4, parameterName=ShapeHeight}
   OutlineThickness
   OutlineAlignment
   GlowMin
   GlowMax
  ) {materialFunction=/FortUILibrary/MaterialFunctions/Shapes/MF_UI_IsoscelesTrapezoid}
  LinearInterpolate(
   A=LinearInterpolate(
    A=MaterialFunctionCall!(BufferSource) {materialFunction=/BattleRoyaleGameUI/Shared/MF_UI_AdaptiveColors}.Dark
    B=VectorParameter() {parameterName=ShadowColor, defaultValue=<rgba 0 0 0 1>}.RGB
    Alpha=ScalarParameter() {parameterName=CustomShadowcolor}
   )
   B=LinearInterpolate(
    A=VectorParameter() {parameterName=Color, defaultValue=<rgba 0.9301 1 0.0103 1>}.RGB
    Alpha=ScalarParameter() {defaultValue=1, parameterName=UseBlue}
   )
   Alpha=@n3.Fill
  )
  n4 = Multiply(
   A=MaterialFunctionCall(
    DistanceField=@n3.SDF
    Threshold=ScalarParameter() {defaultValue=-0.025, parameterName=Shadow_Threshold}
    Smoothing=ScalarParameter() {defaultValue=0.5, parameterName=Shadow_Smoothing}
   ) {materialFunction=&MF_UI_AntiAliasing}.Inner
   B=ScalarParameter() {defaultValue=0.25, parameterName=Shadow_Opacity}
  )
  LinearInterpolate(A=NamedRerouteDeclaration(None=Saturate(None=@n4)) {name=ShadowMask, nodeColor=<rgba 0 0 0 1>}, Alpha=@n3.Fill)
 ]
}
